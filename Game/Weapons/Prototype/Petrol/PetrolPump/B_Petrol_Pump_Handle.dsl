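; Petrol pump handle weapon Blueprint that adds fuel-leak visuals and links its spawned projectiles for a gas trail.
; Each spawned projectile registers as a previous source on the FortSplatterSourceComponent.
; On equip, it attaches a Niagara leak effect to the weapon mesh root and sets its direction and offset vectors.
; Fire FX start and stop toggle that effect, and local unequip destroys it.
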(event FortWeapon|EventOnSpawnProjectile
  (params (SpawnProjectile :type Fort%20Projectile%20Base%20Object%20Reference))
  ;[1 1 1] On Spawn Projectile, set up for spawning gas trail
  (Class|FortSplatterSourceComponent|AddPreviousSource (Variables|Default|GetFortSplatterSource) SpawnProjectile))

(event FortWeapon|EventOnEquippedAndAttached
  (bind _ns_fuelleaking (Variables|Default|GetNS_FuelLeaking))
  ;[1 1 1] FX  Fuel trail
  (|CallParentFunction)
  (bind _returnvalue (Niagara|SpawnSystemAttached "/Game/Weapons/Prototype/Petrol/FX/NS_Petrol_FuelLeaking_Pump.NS_Petrol_FuelLeaking_Pump" (Variables|Visual|GetWeaponMesh) "Root" "0, 0, 0" "0, 0, 0" "KeepRelativeOffset" false false))
  (Variables|Default|SetNS_FuelLeaking _returnvalue)
  (Effects|Components|ParticleSystem|SetVectorParameter _ns_fuelleaking "SpeedDirection" "-1.000000, 0,1.000000")
  (Effects|Components|ParticleSystem|SetVectorParameter _ns_fuelleaking "SocketPositionOffset" "5.000000, 0,-3.000000"))

(event EventOnPlayWeaponFireFX
  (params (bPersistentFire :type Boolean) (bSecondaryFire :type Boolean))
  ;[1 1 1] FX  Fuel trail
  (|CallParentFunction false false)
  (Components|Activation|Activate (Variables|Default|GetNS_FuelLeaking) false))

(event EventOnStopWeaponFireFX
  ;[1 1 1] FX  Fuel trail
  (|CallParentFunction)
  (Components|Activation|Deactivate (Variables|Default|GetNS_FuelLeaking)))

(event FortWeapon|EventOnLocalUnEquip
  ;[1 1 1] FX  Fuel trail
  (|CallParentFunction)
  (Components|DestroyComponent (Variables|Default|GetNS_FuelLeaking)))

(fn ConstructionScript
  (|CallParentFunction))
